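# Animation Blueprint for the VitalPsych Claws glider cosmetic that mirrors the pose of the player's parachute mesh.
# On initialization it casts the owning actor to FortPlayerParachute and caches its ParachuteMesh in MeshToCopy.
# The AnimGraph is a single Copy Pose From Mesh node reading that component into the output pose, with no animation logic of its own.
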AnimBlueprint /BRCosmetics/Gadgets/Assets/VinderTech_GliderChute/Glider_VitalPsych_Claws/Meshes/VitalPsych_Claws_Glider_AnimBP
target_skeleton=&/VitalPsych_Claws_Glider_Skeleton
graphs [
 AnimGraph {nodes [Output Pose(Result=Copy Pose From Mesh(SourceMeshComponent=Get MeshtoCopy() -> (MeshtoCopy:Skeletal Mesh Component Object Reference))) {layerGroup=None}]}
 EventGraph {nodes [
   Event BlueprintInitializeAnimation()
   > n1 = Cast To FortPlayerParachute!(execute, Object=GetOwningActor()) -> (then, CastFailed, AsFort Player Parachute:Fort Player Parachute Object Reference) {is=title K2Node_DynamicCast}
   > Set MeshToCopy(MeshToCopy:Skeletal Mesh Component Object Reference=Get ParachuteMesh!(self:Fort Player Parachute Object Reference=@n1) -> (ParachuteMesh:Skeletal Mesh Component Budgeted Object Reference)) -> (Output_Get:Skeletal Mesh Component Object Reference)
   TryGetPawnOwner()
   Event BlueprintUpdateAnimation()
  ]}
]
variables [MeshtoCopy {type=SkeletalMeshComponent}]
